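# Niagara effect for the WalkingSign emote, named as the prop's spawn burst in the main player's emote FX.
# Seven emitters, suffixed LWE (likely lightweight), layer three smoke, two sparkle and two airglow elements; their renderers and internals are not shown.
# The system script runs SystemState and a Set Variables module, with time-of-day color and float-from-array selection among its dynamic inputs.
# It exposes no user parameters.
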NiagaraSystem /BRCosmetics/Animation/Game/MainPlayer/Emotes/WalkingSign/FX/NS_Emote_WalkingSign_PropSpawn
dependencies {
 usedDataInterfaces [&NiagaraDataInterfaceArrayFloat]
 usedDynamicInputs [&FNTimeOfDayColor, &SelectFloatFromArray]
 usedModules [
  &SystemState
  &self:SystemSpawnScript.SystemScriptSource.SystemScriptGraph.NiagaraNodeAssignment_1.SetVariables_#1
 ]
}
emitters {
 AirglowLWE {rendererClasses [], renderers []}
 AirglowLWE001 {rendererClasses [], renderers []}
 SmokeLWE {rendererClasses [], renderers []}
 SmokeLWE001 {rendererClasses [], renderers []}
 SmokeLWE002 {rendererClasses [], renderers []}
 SparklesLWE {rendererClasses [], renderers []}
 SparklesLWE001 {rendererClasses [], renderers []}
}
summary {
 systemName=NS_Emote_WalkingSign_PropSpawn
 emitters [
  {emitterName=SparklesLWE001, rendererClasses []}
  {emitterName=AirglowLWE001, rendererClasses []}
  {emitterName=SmokeLWE, rendererClasses []}
  {emitterName=SmokeLWE001, rendererClasses []}
  {emitterName=SparklesLWE, rendererClasses []}
  {emitterName=AirglowLWE, rendererClasses []}
  {emitterName=SmokeLWE002, rendererClasses []}
 ]
 userVariables []
}
